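# Material function for the Trifle Parfait player body that adds a timed darken reactivity to emissive, behind a static switch.
# Time since TP_CurrentTime_Reactivity, normalized by duration, drives a cosine ease-in, hold, ease-out pulse declared as ReactivityFading.
# An altered attribute set scales emissive, optionally adding squared eye-glow color, and is blended with unaltered attributes.
# Named reroutes expose the material attributes, SkinFX green as GlowIntensity, and the skin mask.
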MaterialFunction /Game/Characters/Player/Male/Medium/Bodies/M_MED_Trifle/Materials/MF_Trifle_Parfait_Darken_Reactivity
graph {nodes [
  n1 = Time()
  n2 = ScalarParameter() {defaultValue=-99999, group=TrifleParfait, parameterName=TP_CurrentTime_Reactivity, sliderMax=1}
  n3 = ScalarParameter() {defaultValue=4, parameterName=TP_Anim_Duration, sliderMax=4, sliderMin=0}
  n4 = Saturate(None=Divide(A=Subtract(A=@n1, B=@n2), B=Add(A=@n3) {constB=0}))
  Composite!(Time=@n4) {subgraphName=EaseInOut}
  n5 = NamedRerouteUsage()
  n6 = GetMaterialAttributes(None=NamedRerouteUsage()) {attributeGetTypes [#1]}
  n7 = Multiply(A=@n6.EmissiveColor, B=ScalarParameter() {defaultValue=0.8, group=TrifleParfait, parameterName=Scale_Amount, sliderMax=1})
  n8 = VectorParameter() {desc="RGB=Color, A=Intensity", parameterName=EyesGlowColor, defaultValue=<rgba 0 0.4855 1 5>}
  n9 = Multiply(A=Multiply(A=@n8.RGB, B=@n8.RGB), B=NamedRerouteUsage())
  n10 = SetMaterialAttributes!(
   MaterialAttributes=@n6.MaterialAttributes
   EmissiveColor=StaticSwitchParameter(True=Add(A=@n7, B=@n9), False=@n7) {group=TrifleParfait, parameterName=UseTrifleParfait_EyesGlow, sortPriority=1}
  ) {attributeSetTypes [#1]}
  n11 = NamedRerouteUsage()
  FunctionOutput(None=StaticSwitchParameter(True=BlendMaterialAttributes(A=@n5, B=@n10, Alpha=@n11), False=NamedRerouteUsage()) {group=TrifleParfait, parameterName=UseTrifleParfait_DarkenReactivity, sortPriority=0})
  NamedRerouteDeclaration(None=FunctionInput() {inputName=MA, inputType=FunctionInput_MaterialAttributes}) {name=MA, nodeColor=<rgba 1 0.2696 0 1>}
  NamedRerouteDeclaration(None=ComponentMask(None=FunctionInput() {inputName=SkinFX, inputType=FunctionInput_Vector4, sortPriority=1}) {g=true}) {name=GlowIntensity, nodeColor=<rgba 0.1218 1 0 1>}
  MaterialExpressionReroute_7 = Reroute(None=@n4)
  n12 = Multiply(A=@MaterialExpressionReroute_7.Time) {constB=2}
  n13 = Cosine(None=@n12)
  n14 = Add(A=@n12) {constB=0.5}
  n15 = Constant() {r=2}
  n16 = Max(A=Add(A=Multiply(A=@n13) {constB=-0.5}) {constB=0.5}, B=Floor(None=Fmod(A=@n14, B=@n15)))
  MaterialExpressionReroute_8 = Reroute(None=@n16)
  NamedRerouteDeclaration(None=@MaterialExpressionReroute_8.Blend) {name=ReactivityFading, nodeColor=<rgba 0.1719 1 0 1>}
  NamedRerouteDeclaration(None=FunctionInput() {inputName=Skin, inputType=FunctionInput_Scalar, sortPriority=2}) {name=SkinMask, nodeColor=<rgba 0.1016 1 0 1>}
  PinBase() {reroutePins [Time {expression=&self:MaterialExpressionReroute_7}]}
  PinBase!(Blend=@n16) {reroutePins [Blend {expression=&self:MaterialExpressionReroute_8}]}
 ]}
