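# Master character material for the Trifle Parfait player body, using material attributes and per-expression shading models.
# Base PreFX output, skin mask and Gmap are shared through named reroutes, alongside a SkinFX mask texture.
# Static switches layer optional metal LUT, silk, cloth fuzz and clear coat, followed by anisotropic shading.
# A Parfait-specific darken reactivity function feeds the shared character PostFX, then cinematic post FX.
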Material /Game/Characters/Player/Male/Medium/Bodies/M_MED_Trifle/Materials/M_Trifle_Parfait_MASTER
graph {
 properties {
  bAutomaticallySetUsageInEditor=false
  bUseMaterialAttributes=true
  bUsedWithClothing=true
  bUsedWithMeshDeformer=true
  bUsedWithMorphTargets=true
  bUsedWithNanite=true
  bUsedWithSkeletalMesh=true
  pixelDepthOffsetMode=PDOM_Legacy
  refractionMethod=RM_IndexOfRefraction
  shadingModel=MSM_FromMaterialExpression
  subsurfaceProfile=&SS_HeroSkin_02
  usedShadingModels=Default Lit | Subsurface | Clear Coat | Subsurface Profile | Cloth | Eye
 }
 nodes [
  n1 = MaterialFunctionCall() {materialFunction=&BaseFN_CharacterFunction_PreFX}
  n2 = MaterialFunctionCall(
   AO=@n1.AO
   Distance Blend=@n1.AO
   Fresnel=@n1.AO
   Gmap=@n1.AO
   Pre Skinned Local Position=@n1.AO
   Material=MaterialFunctionCall!(MA=NamedRerouteUsage(), SkinFX=NamedRerouteUsage(), Skin=NamedRerouteUsage()) {materialFunction=&/MF_Trifle_Parfait_Darken_Reactivity}.Result
  ) {materialFunction=&BaseFN_CharacterFunction_PostFX}
  MaterialFunctionCall(In=@n2.Result) {materialFunction=&CINE_MF_CharPostFX}
  MaterialFunctionCall(
   PreskinnedPosition=@n1.Pre skinned Local Position
   Gmap=@n1.Pre skinned Local Position
   MaterialAttributes=@n1.Pre skinned Local Position
   AO=@n1.Pre skinned Local Position
  ) {materialFunction=&CharacterSkinFX}
  n3 = TextureSampleParameter2D() {group=Skin FX, parameterName=SkinFX_Mask, samplerType=SAMPLERTYPE_Masks, texture=&T_BlackMask}
  NamedRerouteDeclaration(None=@n3.RGBA) {name=SkinFX, nodeColor=<rgba 1 0.6094 0 1>}
  NamedRerouteDeclaration(None=@n1.Result) {name=MA_PreFX, nodeColor=<rgba 0 0.4141 1 1>}
  NamedRerouteDeclaration(None=@n1.Gmap) {name=Gmap, nodeColor=<rgba 0 1 0.7188 1>}
  n4 = NamedRerouteUsage()
  n5 = NamedRerouteUsage()
  n6 = NamedRerouteUsage()
  n7 = StaticSwitchParameter(
   True=MaterialFunctionCall(Mask=NamedRerouteUsage(), Materials=@n6) {materialFunction=&MF_MetalsLUT}.Result
   False=@n6
  ) {group=Skin FX, parameterName=UseMetalLUT}
  n8 = StaticSwitchParameter(
   True=MaterialFunctionCall!(Alpha=ChannelMaskParameter(None=@n5) {group=Skin FX, parameterName=Silk Channel}.R, Material=@n7) {materialFunction=/Game/Weapons/WeaponSkins/MaterialFunction/MF_Silk}.Result
   False=@n7
  ) {group=Skin FX, parameterName=Use Silk}
  n9 = StaticSwitchParameter(
   True=MaterialFunctionCall!(
    Alpha=StaticComponentMaskParameter(None=@n4) {defaultR=true, group=Skin FX, parameterName=Cloth Channel}
    Gmap=NamedRerouteUsage()
    Material=@n8
   ) {materialFunction=/Game/Weapons/WeaponSkins/MaterialFunction/ClothFuzz}.Result
   False=@n8
  ) {group=Skin FX, parameterName=Use Cloth fuzz}
  n10 = MaterialFunctionCall!(Alpha=StaticComponentMaskParameter(None=NamedRerouteUsage()) {defaultR=true, group=Skin FX, parameterName=Clear Coat Channel}, In=@n9) {materialFunction=/Game/Weapons/WeaponSkins/MaterialFunction/Clearcoat}
  NamedRerouteDeclaration(None=MaterialFunctionCall!(In=StaticSwitchParameter(True=@n10.Result, False=@n9) {group=Skin FX, parameterName=Use Clear Coat}) {materialFunction=/Game/Weapons/WeaponSkins/MaterialFunction/MF_AnisotropicShading}.Result) {name=MA_PostFX, nodeColor=<rgba 0 0.6484 1 1>}
  NamedRerouteDeclaration(None=@n1.Skin) {name=SkinMask, nodeColor=<rgba 1 0.3604 0.4302 1>}
 ]
}
